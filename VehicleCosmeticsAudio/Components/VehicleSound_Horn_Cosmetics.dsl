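; Vehicle audio component that plays a vehicle's horn, using the cosmetic horn sound or a default fallback.
; On init it rolls a random pitch offset and binds horn and rev limiter events, unbinding them on shutdown.
; Horn presses spawn attached audio with spam and stacking options, and releases fade it out.
; With redline-when-held enabled, holding below direction-specific minimum speeds revs the engine instead, and rev limiter hits above them clear redline.
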(event VehicleAudio|EventOnInit
  (bind _returnvalue (VehicleSound|GetAudioController))
  (bind _returnvalue_2 (Actor|GetComponentbyClass _returnvalue "/Script/AudioMotorSimStandardComponents.RevLimiterMotorSimComponent"))
  (bind _returnvalue_3 (VehicleSound|GetVehicleActor))
  (bind (_hornpitchoffset_x _hornpitchoffset_y) (Math|Vector2D|BreakVector2D (Variables|GetHornPitchOffset_X)))
  (|CallParentFunction)
  (Utilities|FlowControl|Sequence
    (:then_0
      (Utilities|IsValid _returnvalue_3
        (:Is%20Valid
          (Vehicle|BindEventtoOnVehicleHornEvent _returnvalue_3 (EventDispatchers|CreateEvent :!event OnHorn)))))
    (:then_1
      (Utilities|IsValid _returnvalue_2
        (:Is%20Valid
          (EventDispatchers|BindEventtoOnRevLimiterHit _returnvalue_2 (EventDispatchers|CreateEvent :!event OnRevLimiterHit)))))
    (:then_2
      (Variables|Horn|Private|SetPitchOffset (Math|Random|RandomFloatinRange _hornpitchoffset_x _hornpitchoffset_y)))))

(event VehicleAudio|EventOnShutdown
  (bind _returnvalue (VehicleSound|GetVehicleActor))
  (bind _returnvalue_2 (VehicleSound|GetAudioController))
  (bind _returnvalue_3 (Actor|GetComponentbyClass _returnvalue_2 "/Script/AudioMotorSimStandardComponents.RevLimiterMotorSimComponent"))
  (|CallParentFunction)
  (Utilities|FlowControl|Sequence
    (:then_0
      (Utilities|IsValid _returnvalue_3
        (:Is%20Valid
          (EventDispatchers|UnbindEventfromOnRevLimiterHit _returnvalue_3 (EventDispatchers|CreateEvent :!event OnRevLimiterHit)))))
    (:then_1
      (Utilities|IsValid _returnvalue
        (:Is%20Valid
          (Vehicle|UnbindEventfromOnVehicleHornEvent _returnvalue (EventDispatchers|CreateEvent :!event OnHorn)))))))

(fn CanUseHorn
  (return (or (and (Variables|Redline|GetbShouldRedlineWhenHeld) (>= (Vehicle|GetTotalSpeedKmh (VehicleSound|GetVehicleActor)) (Math|Float|SelectFloat (Variables|Horn|GetHornMinSpeed) (Variables|Horn|GetReverse%20Horn%20Min%20Speed) (> (Vehicle|GetForwardSpeedKmh (VehicleSound|GetVehicleActor)) 0)))) (not (Variables|Redline|GetbShouldRedlineWhenHeld)))))

(fn GetHornSound
  (bind _asbp_vehicle_cosmetics_audio_controller_parent (Utilities|Casting|CastToBP_VehicleCosmetics_AudioController_Parent (VehicleSound|GetAudioController))
    (:then
      (bind (_cosmeticdata_enginesound _cosmeticdata_engineignitionsound _cosmeticdata_engineoffsound _cosmeticdata_gearshiftupsound _cosmeticdata_gearshiftdownsound _cosmeticdata_hornsound) (Utilities|Struct|BreakVehicleAudioCosmeticDataStruct (Class|BPVehicleCosmeticsAudioControllerParent|GetCosmeticDataStruct _asbp_vehicle_cosmetics_audio_controller_parent)))
      (Utilities|IsValid _cosmeticdata_hornsound
        (:Is%20Valid
          (return _cosmeticdata_hornsound))
        (:Is%20Not%20Valid
          _)))
    (:CastFailed
      _))
  (return (Variables|Horn|GetDefaultHornSound)))

(fn OnHorn
  (params (Controller :type Fort%20Player%20Controller%20Athena%20Object%20Reference) (Vehicle :type Fort%20Athena%20Vehicle%20Object%20Reference) (bIsHeld :type Boolean))
  (Variables|Horn|Private|SetIsHeld bIsHeld)
  (Utilities|FlowControl|Sequence
    (:then_0
      ;[1 1 1] Set state of horn if it is been held or not.
      (Variables|Horn|GetDefaultHornSound
        (:then
          (if (Variables|Horn|Private|GetIsHeld)
            (bind _hornaudio (Variables|Default|GetHornAudio)
              (:then
                ;[1 1 1] Are we already playing, or do we want this to be spammable
                (if (not (Math|Boolean|NORBoolean (Audio|Components|Audio|IsPlaying _hornaudio) (not (Variables|Horn|GetbIsSpammable))))
                  ;[1 1 1] Do we want to stack voice or just kill older.
                  (if (not (Variables|Horn|GetbStackable))
                    (bind _hornaudio_2 (Variables|Default|GetHornAudio)
                      (:then
                        (Audio|Components|Audio|FadeOut _hornaudio_2 0.25 0.0 "Linear" :!class /Script/Engine.AudioComponent))
                      (:else)))))
              (:else
                _))
            ;[1 1 1] Play horn sound if able, with pitch offset
            (if (CallFunction|CanUseHorn)
              (bind _returnvalue (Audio|SpawnSoundAttached (CallFunction|GetHornSound) (Variables|/Script/Engine.Actor|GetRootComponent (Components|GetOwner))))
              (bind _output_get (Variables|Default|SetHornAudio _returnvalue))
              (Audio|Parameter|SetFloatParameter _output_get "PitchOffset" (Variables|Horn|Private|GetPitchOffset) :!class /Script/Engine.AudioComponent))
            (else
              ;[1 1 1] Stop On Release check is removed to allow all horns to be looping. CL 53140504 all horns are stop on release.
              (bind _hornaudio_3 (Variables|Default|GetHornAudio)
                (:then
                  (Audio|Components|Audio|FadeOut _hornaudio_3 (Variables|Horn|GetHornFadeOutOnRelease) 0.0 "Linear" :!class /Script/Engine.AudioComponent))
                (:else)))))
        (:else)))
    (:then_1
      (if (and (not (CallFunction|CanUseHorn)) (Variables|Redline|GetbShouldRedlineWhenHeld))
        (AudioController|SetRedlineActive (VehicleSound|GetAudioController) (Variables|Horn|Private|GetIsHeld))))))

(fn OnRevLimiterHit
  (if (CallFunction|CanUseHorn)
    (AudioController|SetRedlineActive (VehicleSound|GetAudioController) false)))
