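; Vehicle audio component that plays gear shift, reversing and rev limiter sounds.
; It binds to the audio controller's motor sim gear change and rev limiter events on init, unbinding on shutdown.
; Shift sounds prefer the controller's cosmetic data, using defaults unless bIsIP is set, and fade out the previous shift.
; Reverse audio spawns on the engine component after a delay and fades out; limiter hits play 2D for local occupants.
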(event VehicleAudio|EventOnInit
  (bind _returnvalue (VehicleSound|GetAudioController))
  (|CallParentFunction)
  (PhysicsSim|AssignOnGearChangedEvent (Actor|GetComponentbyClass _returnvalue "/Script/AudioMotorSimStandardComponents.MotorPhysicsSimComponent") (EventDispatchers|CreateEvent :!event OnGearShiftChange))
  (EventDispatchers|BindEventtoOnRevLimiterHit (Actor|GetComponentbyClass _returnvalue "/Script/AudioMotorSimStandardComponents.RevLimiterMotorSimComponent") (EventDispatchers|CreateEvent :!event OnRevLimiterHit)))

(event VehicleAudio|EventOnShutdown
  (bind _returnvalue (VehicleSound|GetAudioController))
  (|CallParentFunction)
  (PhysicsSim|UnbindEventfromOnGearChangedEvent (Actor|GetComponentbyClass _returnvalue "/Script/AudioMotorSimStandardComponents.MotorPhysicsSimComponent") (EventDispatchers|CreateEvent :!event OnGearShiftChange))
  (EventDispatchers|UnbindEventfromOnRevLimiterHit (Actor|GetComponentbyClass _returnvalue "/Script/AudioMotorSimStandardComponents.RevLimiterMotorSimComponent") (EventDispatchers|CreateEvent :!event OnRevLimiterHit)))

(event EventOnReversedChanged
  (params (bReversing :type Boolean))
  (if bReversing
    (Utilities|FlowControl|Delay (Variables|Reverse|GetReversingTriggerTime))
    (if bReversing
      (Variables|Default|GetReverseAudio
        (:then)
        (:else
          (bind _reversesound (Variables|Reverse|GetReverseSound)
            (:then
              (bind _returnvalue (Audio|SpawnSoundAttached _reversesound (AudioController|GetEngineAudioComponent (VehicleSound|GetAudioController))))
              (Variables|Default|SetReverseAudio _returnvalue))
            (:else)))))
    (else
      (bind _reverseaudio (Variables|Default|GetReverseAudio)
        (:then
          (Audio|Components|Audio|FadeOut _reverseaudio (Variables|Reverse|GetReverseFadeOutDuration) 0.0 "Linear" :!class /Script/Engine.AudioComponent))
        (:else)))))

(fn OnGearShiftChange
  (params (NewGear :type Integer))
  (bind _asbp_vehicle_cosmetics_audio_controller_parent (Utilities|Casting|CastToBP_VehicleCosmetics_AudioController_Parent (VehicleSound|GetAudioController))
    (:then
      (bind (_cosmeticdata_enginesound _cosmeticdata_engineignitionsound _cosmeticdata_engineoffsound _cosmeticdata_gearshiftupsound _cosmeticdata_gearshiftdownsound _cosmeticdata_hornsound _cosmeticdata_bisip) (Utilities|Struct|BreakVehicleAudioCosmeticDataStruct (Class|BPVehicleCosmeticsAudioControllerParent|GetCosmeticDataStruct _asbp_vehicle_cosmetics_audio_controller_parent)))
      (CallFunction|PlayGearSound NewGear (select (Utilities|IsValid _cosmeticdata_gearshiftupsound :!class /Script/Engine.KismetSystemLibrary) _cosmeticdata_gearshiftupsound (select _cosmeticdata_bisip 0 (Variables|Gear|GetGearShiftUpSound))) (select (Utilities|IsValid _cosmeticdata_gearshiftdownsound :!class /Script/Engine.KismetSystemLibrary) _cosmeticdata_gearshiftdownsound (select _cosmeticdata_bisip 0 (Variables|Gear|GetGearShiftDownSound)))))
    (:CastFailed
      (CallFunction|PlayGearSound NewGear (Variables|Gear|GetGearShiftUpSound) (Variables|Gear|GetGearShiftDownSound))))
  (Variables|Private|SetCurrentGear (Variables|GetNewGear)))

(fn OnRevLimiterHit
  (bind _asbp_vehicle_cosmetics_audio_controller_parent (Utilities|Casting|CastToBP_VehicleCosmetics_AudioController_Parent (VehicleSound|GetAudioController))
    (:then
      (bind (_cosmeticdata_enginesound _cosmeticdata_engineignitionsound _cosmeticdata_engineoffsound _cosmeticdata_gearshiftupsound _cosmeticdata_gearshiftdownsound _cosmeticdata_hornsound _cosmeticdata_bisip) (Utilities|Struct|BreakVehicleAudioCosmeticDataStruct (Class|BPVehicleCosmeticsAudioControllerParent|GetCosmeticDataStruct _asbp_vehicle_cosmetics_audio_controller_parent)))
      (if _cosmeticdata_bisip
        _
        (else
          _)))
    (:CastFailed
      _))
  (if (Variables|/VehicleCosmeticsAudio/Controllers/BP_VehicleCosmetics_AudioController_Parent.BP_VehicleCosmetics_AudioController_Parent_C|GetIsLocalPawnInVehicle _asbp_vehicle_cosmetics_audio_controller_parent)
    (bind _revlimiterhitsound (Variables|Limiter|GetRevLimiterHitSound)
      (:then
        (Audio|PlaySound2D _revlimiterhitsound))
      (:else))))

(fn PlayGearSound
  (params (NewGearIndex :type Integer) (UpshiftSound :type Sound%20Base%20Object%20Reference) (DownshiftSound :type Sound%20Base%20Object%20Reference))
  (bind _currentgear (Variables|Private|GetCurrentGear))
  (bind _returnvalue (> NewGearIndex _currentgear))
  (bind _returnvalue_2 (select _returnvalue UpshiftSound DownshiftSound))
  (Utilities|IsValid _returnvalue_2
    (:Is%20Valid
      (bind _shiftaudio (Variables|Default|GetShiftAudio)
        (:then
          (Audio|Components|Audio|FadeOut _shiftaudio 0.25 0.0 "Linear" :!class /Script/Engine.AudioComponent))
        (:else
          _))
      (bind _returnvalue_3 (Audio|SpawnSoundAttached _returnvalue_2 (AudioController|GetEngineAudioComponent (VehicleSound|GetAudioController))))
      (Variables|Default|SetShiftAudio _returnvalue_3))))
